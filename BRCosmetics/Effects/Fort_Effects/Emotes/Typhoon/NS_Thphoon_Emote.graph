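# Niagara system for the Typhoon emote that sets a character's hand and lower arm on fire.
# An opening fire and glow burst accompanies a lingering hand glow, both glows carrying lights, plus continuous flame sprites and curl-noise embers.
# Emitters locate particles through the User.SkeletalMesh bone or socket sampling, with lower-arm flames re-sampled each update to slide along the limb.
# Flame sprites animate SubImageIndex and dynamic material parameters over life.
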NiagaraSystem /BRCosmetics/Effects/Fort_Effects/Emotes/Typhoon/NS_Thphoon_Emote
dependencies {
 usedDataInterfaces [&NiagaraDataInterfaceCurve, &NiagaraDataInterfaceColorCurve, &NiagaraDataInterfaceVectorCurve]
 usedDynamicInputs []
 usedModules [
  &self:SystemSpawnScript.SystemScriptSource.SystemScriptGraph.NiagaraNodeAssignment_0.SetVariables_#1
  &SystemState
 ]
}
emitters {
 Embers_Hand {
  emitterUpdateScript [
   &EmitterState(Delay First Loop Only=true, Life Cycle Mode=::NewEnumerator1, Loop Delay=1, Loop Duration=1, UseLoopDelay=true)
   &SpawnBurst_Instantaneous(Spawn Count=300, Use Spawn Probability=false) {enabled=false}
   &SpawnRate(SpawnRate=10, Use Spawn Probability=false)
  ]
  particleSpawnScript [
   /Niagara/Modules/Spawn/Initialization/InitializeParticle(
    Color=<rgba 40 9.2708 3.125 1>
    Lifetime=&UniformRangedFloat(Minimum=0.35, Maximum=0.7)
    Module.Write Color=true
    Module.Write Lifetime=true
    Module.Write Mass=true
    Module.Write Material Random=true
    Module.Write Position=true
    Module.Write Scale=false
    Module.Write SpriteRotation=false
    Module.Write SpriteSize=true
    Position=&SimulationPosition()
    Sprite Size=&Vector2DFromFloat(Value=&FloatFromCurve(
      FloatCurve=DataInterface() {propertyValues {Curve {keys [<at 0 1.2>,<at 0.1257 0.9847>,<at 0.2996 0.5>,<at 1 0.3>]}}}
      CurveIndex=&UniformRangedFloat(Minimum=500, Maximum=1000)
      Scale Curve=3.5
     ))
   )
   &AddVelocityFromPoint(
    Module.Use Falloff Distance=false
    Origin Offset=<0 0 -10>
    Velocity Strength=&RandomRangeFloat(Minimum=75, Maximum=120)
   )
   &SkeletalMeshLocation(
    Bone Sampling Mode=::NewEnumerator4
    Inherit Velocity=false
    Interpolate Mesh Sampling=false
    MaskSpawnGroup=false
    Sampled Position Offset=&RandomRangeVector(Minimum=<-6 -6 -7>, Maximum=<6 6 -6>)
    Sampled Position Offset Coordinate Space:ENiagaraMeshTransforms=::NewEnumerator3
    Skeletal Mesh=$User.SkeletalMesh
    Use Sampled Mesh Attribute as Ribbon ID=false
    Write Sampled Mesh Attrributes to Particle Attributes=true
   )
  ]
  particleUpdateScript [
   &ParticleState()
   &AccelerationForce(Acceleration=<0 0 50>)
   /Niagara/Modules/Update/Forces/CurlNoiseForce(Noise Frequency=0.5, Noise Strength=500, Pan Noise=true, Pan Noise Field=<1 1 1>)
   &SolveForcesAndVelocity(Clamp Velocity=false, Limit Acceleration=false)
   &ScaleColor(
    Scale RGB=&MakeVectorFromLinearColorRGB(Color=&ColorFromCurve(ColorCurve=DataInterface() {propertyValues {
        AlphaCurve {keys [<at 0 1 RCIM_Cubic>]}
        BlueCurve {keys [<at 0 12.2396 RCIM_Cubic>,<at 0.2499 0.8507>,<at 0.7114 0.1788 RCIM_Cubic>]}
        GreenCurve {keys [<at 0 35.8364 RCIM_Cubic>,<at 0.2499 4.9272>,<at 0.7114 0.2017 RCIM_Cubic>]}
        RedCurve {keys [<at 0 50 RCIM_Cubic>,<at 0.2499 10>,<at 0.7114 1 RCIM_Cubic>]}
       }}))
    ScaleA=true
    ScaleRGB=true
   ) {enabled=false}
   &ScaleSpriteSizeBySpeed(
    Max Scale Factor=<2 5>
    Sample Scale Factor By Curve=true
    Scale Factor Curve=DataInterface() {propertyValues {Curve {keys [
        <at -0.0077 0.0765 RCIM_Cubic>
        {arriveTangent=1.2463, interpMode=RCIM_Cubic, leaveTangent=1.2463, tangentMode=RCTM_User, time=0.9985, value=0.9549}
       ]}}}
    Velocity Threshold=500
   )
   &ScaleColor(
    Scale Alpha=&FloatFromCurve(FloatCurve=DataInterface() {propertyValues {Curve {keys [
         <at 0 0 RCIM_Cubic>
         {arriveTangent=0.9268, interpMode=RCIM_Cubic, leaveTangent=0.9268, time=0.27, value=1}
         {arriveTangent=-1.365, interpMode=RCIM_Cubic, leaveTangent=-1.365, time=0.7207, value=0.668}
         <at 1.0011 0.002>
        ]}}})
    Scale RGB=&VectorFromCurve(VectorCurve=DataInterface() {propertyValues {
       XCurve {keys [<at 0 0>, <at 0.2 1.0067>, {arriveTangent=-1.2584, interpMode=RCIM_Cubic, leaveTangent=-1.2584, time=0.8, value=0.9431}, <at 1 0>]}
       YCurve {keys [<at 0 0>, <at 0.2 1.0067>, {arriveTangent=-1.2584, interpMode=RCIM_Cubic, leaveTangent=-1.2584, time=0.8, value=0.4309}, <at 1 0>]}
       ZCurve {keys [<at 0 0>, <at 0.2 1.0067>, {arriveTangent=-1.2584, interpMode=RCIM_Cubic, leaveTangent=-1.2584, time=0.8, value=0}, <at 1 0>]}
      }})
    ScaleA=true
    ScaleRGB=true
   ) {module=ScaleColor001}
  ]
  rendererClasses [&NiagaraSpriteRendererProperties]
  renderers [{rendererClass=&NiagaraSpriteRendererProperties}]
 }
 Fire_Burst {
  emitterUpdateScript [
   &EmitterState(Life Cycle Mode=::NewEnumerator1, Loop Behavior=::NewEnumerator1, Loop Delay=1, Loop Duration=1, UseLoopDelay=true)
   &SpawnBurst_Instantaneous(Spawn Count=3, Spawn Time=0.08, Use Spawn Probability=false)
  ]
  particleSpawnScript [
   &InitializeParticle(
    Color=<rgba 40 9.2708 3.125 1>
    Lifetime=&UniformRangedFloat(Minimum=0.45, Maximum=0.5)
    Module.Write Color=true
    Module.Write Lifetime=true
    Module.Write Mass=true
    Module.Write Material Random=true
    Module.Write Position=false
    Module.Write Scale=false
    Module.Write SpriteRotation=true
    Module.Write SpriteSize=true
    Position=&SimulationPosition()
    Sprite Rotation=&UniformRangedFloat(Minimum=0, Maximum=360)
    Sprite Size=&Vector2DFromFloat(Value=&UniformRangedFloat(Minimum=55, Maximum=65))
   )
   &DynamicMaterialParameters(
    Index 0 Param 2=&UniformRangedFloat(Minimum=0, Maximum=1)
    Index 0 Param 4=&UniformRangedFloat(Minimum=3, Maximum=5)
    Module.Param0WriteEnabled=false
    Module.Param1WriteEnabled=true
    Module.Param2WriteEnabled=false
    Module.Param3WriteEnabled=true
   )
   &AddVelocityFromPoint(Module.Use Falloff Distance=false, Velocity Origin:Vector3f=&SimulationPosition(), Velocity Strength=42)
   &AddVelocity(Velocity=<0 0 60>)
   &SkeletalMeshLocation(
    Bone Sampling Mode=::NewEnumerator4
    Inherit Velocity=false
    Interpolate Mesh Sampling=false
    MaskSpawnGroup=false
    Sampled Position Offset=&RandomRangeVector(Minimum=<-3 -3 0>, Maximum=<3 3 6>)
    Sampled Position Offset Coordinate Space:ENiagaraMeshTransforms=::NewEnumerator3
    Skeletal Mesh=$User.SkeletalMesh
    Use Sampled Mesh Attribute as Ribbon ID=false
    Write Sampled Mesh Attrributes to Particle Attributes=true
   )
  ]
  particleUpdateScript [
   &ParticleState()
   &ScaleSpriteSize(Scale Factor=&Vector2DFromFloat(Value=&FloatFromCurve(FloatCurve=DataInterface() {propertyValues {Curve {keys [
          {arriveTangent=2.4033, interpMode=RCIM_Cubic, leaveTangent=2.4033, tangentMode=RCTM_User, time=0.0002, value=0.15}
          {arriveTangent=-0.0027, interpMode=RCIM_Cubic, leaveTangent=-0.0027, tangentMode=RCTM_User, time=0.6253, value=0.7425}
          {arriveTangent=-0.1019, interpMode=RCIM_Cubic, leaveTangent=-0.1019, tangentMode=RCTM_User, time=1.0041, value=0.7174}
         ]}}})))
   /SetVariables(Particles.SubImageIndex=&FloatFromCurve(
     FloatCurve=DataInterface() {propertyValues {Curve {keys [
         {arriveTangent=2.1444, interpMode=RCIM_Cubic, leaveTangent=2.1444, tangentMode=RCTM_User, time=0, value=0}
         {arriveTangent=0.1938, interpMode=RCIM_Cubic, leaveTangent=0.1938, tangentMode=RCTM_User, time=1, value=1}
        ]}}}
     Scale Curve=40
    )) {isSetParameters=true, module=SetVariables_#2}
   &DynamicMaterialParameters(
    Index 0 Param 1=&FloatFromCurve(FloatCurve=DataInterface() {propertyValues {Curve {keys [<at 0 0>,<at 0.8 1>,<at 1.0011 1>]}}})
    Index 0 Param 2=2.5
    Index 0 Param 3=1
    Index 0 Param 4=3
    Module.Param0WriteEnabled=true
    Module.Param1WriteEnabled=false
    Module.Param2WriteEnabled=true
    Module.Param3WriteEnabled=false
   ) {module=DynamicMaterialParameters001}
   &GravityForce(Gravity=&MakeVector(
     X=0
     Y=0
     Z=&FloatFromCurve(FloatCurve=DataInterface() {propertyValues {Curve {keys [<at 0 0>,<at 0.6 0>,<at 0.7124 0.9779>,<at 1 1>]}}}, Scale Curve=250)
    )) {enabled=false}
   &Drag(Drag=0.7, Use Linear Drag=true, Use Rotational Drag=false)
   &SolveForcesAndVelocity(Clamp Velocity=false, Limit Acceleration=false)
   &ScaleColor(
    Scale RGB=&MakeVectorFromLinearColorRGB(Color=&ColorFromCurve(ColorCurve=DataInterface() {propertyValues {
        AlphaCurve {keys [<at 0 1 RCIM_Cubic>]}
        BlueCurve {keys [<at 0 0.8073 RCIM_Cubic>,<at 0.1964 0.4479>,<at 0.8699 0.0677 RCIM_Cubic>]}
        GreenCurve {keys [<at 0 2.1892 RCIM_Cubic>,<at 0.1964 1.1205>,<at 0.8699 0.1173 RCIM_Cubic>]}
        RedCurve {keys [<at 0 3 RCIM_Cubic>,<at 0.1964 2>,<at 0.8699 1 RCIM_Cubic>]}
       }}))
    ScaleA=true
    ScaleRGB=true
   ) {enabled=false}
   &SpriteRotationRate(Rotation Rate=150)
   &ScaleColor(
    Scale Alpha=&FloatFromCurve(FloatCurve=DataInterface() {propertyValues {Curve {keys [
         <at 0 0 RCIM_Cubic>
         {arriveTangent=0.9268, interpMode=RCIM_Cubic, leaveTangent=0.9268, time=0.27, value=1}
         {arriveTangent=-1.365, interpMode=RCIM_Cubic, leaveTangent=-1.365, time=0.7207, value=0.668}
         <at 1.0011 0.002>
        ]}}})
    Scale RGB=&VectorFromCurve(VectorCurve=DataInterface() {propertyValues {
       XCurve {keys [<at 0 0>, <at 0.2 1.0067>, {arriveTangent=-1.2584, interpMode=RCIM_Cubic, leaveTangent=-1.2584, time=0.8, value=0.9431}, <at 1 0>]}
       YCurve {keys [<at 0 0>, <at 0.2 1.0067>, {arriveTangent=-1.2584, interpMode=RCIM_Cubic, leaveTangent=-1.2584, time=0.8, value=0.4309}, <at 1 0>]}
       ZCurve {keys [<at 0 0>, <at 0.2 1.0067>, {arriveTangent=-1.2584, interpMode=RCIM_Cubic, leaveTangent=-1.2584, time=0.8, value=0}, <at 1 0>]}
      }})
    ScaleA=true
    ScaleRGB=true
   ) {module=ScaleColor001}
  ]
  rendererClasses [&NiagaraSpriteRendererProperties]
  renderers [{rendererClass=&NiagaraSpriteRendererProperties}]
 }
 Fire_Decal {
  emitterUpdateScript [
   &EmitterState(Delay First Loop Only=true, Life Cycle Mode=::NewEnumerator1, Loop Delay=1, Loop Duration=1, UseLoopDelay=true)
   &SpawnBurst_Instantaneous(Spawn Count=8, Spawn Time=0.08, Use Spawn Probability=false) {enabled=false}
   &SpawnRate(SpawnRate=4, Use Spawn Probability=false)
  ]
  particleSpawnScript [
   &InitializeParticle(
    Color=<rgba 40 9.2708 3.125 1>
    Lifetime=&UniformRangedFloat(Minimum=0.45, Maximum=0.5)
    Module.Write Color=true
    Module.Write Lifetime=true
    Module.Write Mass=true
    Module.Write Material Random=true
    Module.Write Position=false
    Module.Write Scale=false
    Module.Write SpriteRotation=true
    Module.Write SpriteSize=true
    Position=&SimulationPosition()
    Sprite Rotation=&UniformRangedFloat(Minimum=0, Maximum=360)
    Sprite Size=&Vector2DFromFloat(Value=&UniformRangedFloat(Minimum=50, Maximum=70))
   )
   &DynamicMaterialParameters(
    Index 0 Param 2=&UniformRangedFloat(Minimum=0, Maximum=1)
    Index 0 Param 4=&UniformRangedFloat(Minimum=3, Maximum=5)
    Module.Param0WriteEnabled=false
    Module.Param1WriteEnabled=true
    Module.Param2WriteEnabled=false
    Module.Param3WriteEnabled=true
   )
   &AddVelocityFromPoint(
    Module.Use Falloff Distance=false
    Origin Offset=<0 0 -10>
    Velocity Origin:Vector3f=&SimulationPosition()
    Velocity Strength=80
   )
   &SkeletalMeshLocation(
    Bone Sampling Mode=::NewEnumerator4
    Inherit Velocity=false
    Interpolate Mesh Sampling=false
    MaskSpawnGroup=false
    Sampled Position Offset=&RandomRangeVector(Minimum=<-3 -3 -7>, Maximum=<3 3 -6>)
    Skeletal Mesh=$User.SkeletalMesh
    Use Sampled Mesh Attribute as Ribbon ID=false
    Write Sampled Mesh Attrributes to Particle Attributes=true
   )
   /SetVariables(Particles.Sprite2_Size:Vector2f=<33 33>) {isSetParameters=true}
  ]
  particleUpdateScript [
   &ParticleState()
   &ScaleSpriteSize(Scale Factor=&Vector2DFromFloat(Value=&FloatFromCurve(FloatCurve=DataInterface() {propertyValues {Curve {keys [
          {arriveTangent=2.4033, interpMode=RCIM_Cubic, leaveTangent=2.4033, tangentMode=RCTM_User, time=0.0002, value=0.15}
          {arriveTangent=-0.0027, interpMode=RCIM_Cubic, leaveTangent=-0.0027, tangentMode=RCTM_User, time=0.6253, value=0.7425}
          {arriveTangent=-0.1019, interpMode=RCIM_Cubic, leaveTangent=-0.1019, tangentMode=RCTM_User, time=1.0041, value=0.7174}
         ]}}})))
   /SetVariables(Particles.SubImageIndex=&FloatFromCurve(
     FloatCurve=DataInterface() {propertyValues {Curve {keys [
         {arriveTangent=2.1444, interpMode=RCIM_Cubic, leaveTangent=2.1444, tangentMode=RCTM_User, time=0, value=0}
         {arriveTangent=0.1938, interpMode=RCIM_Cubic, leaveTangent=0.1938, tangentMode=RCTM_User, time=1, value=1}
        ]}}}
     Scale Curve=40
    )) {isSetParameters=true, module=SetVariables_#2}
   &DynamicMaterialParameters(
    Index 0 Param 1=&FloatFromCurve(FloatCurve=DataInterface() {propertyValues {Curve {keys [<at 0 0>,<at 0.8 1>,<at 1.0011 1>]}}})
    Index 0 Param 2=2.5
    Index 0 Param 3=1
    Index 0 Param 4=3
    Module.Param0WriteEnabled=true
    Module.Param1WriteEnabled=false
    Module.Param2WriteEnabled=true
    Module.Param3WriteEnabled=false
   ) {module=DynamicMaterialParameters001}
   &GravityForce(Gravity=&MakeVector(
     X=0
     Y=0
     Z=&FloatFromCurve(FloatCurve=DataInterface() {propertyValues {Curve {keys [<at 0 0>,<at 0.6 0>,<at 0.7124 0.9779>,<at 1 1>]}}}, Scale Curve=250)
    ))
   &Drag(Drag=2.5, Use Linear Drag=true, Use Rotational Drag=false) {enabled=false}
   &SolveForcesAndVelocity(Clamp Velocity=false, Limit Acceleration=false)
   &ScaleColor(
    Scale RGB=&MakeVectorFromLinearColorRGB(Color=&ColorFromCurve(ColorCurve=DataInterface() {propertyValues {
        AlphaCurve {keys [<at 0 1 RCIM_Cubic>]}
        BlueCurve {keys [<at 0 0.8073 RCIM_Cubic>,<at 0.1964 0.4479>,<at 0.8699 0.0677 RCIM_Cubic>]}
        GreenCurve {keys [<at 0 2.1892 RCIM_Cubic>,<at 0.1964 1.1205>,<at 0.8699 0.1173 RCIM_Cubic>]}
        RedCurve {keys [<at 0 3 RCIM_Cubic>,<at 0.1964 2>,<at 0.8699 1 RCIM_Cubic>]}
       }}))
    ScaleA=true
    ScaleRGB=true
   ) {enabled=false}
   &SpriteRotationRate(Rotation Rate=150)
   &ScaleColor(
    Scale Alpha=&FloatFromCurve(FloatCurve=DataInterface() {propertyValues {Curve {keys [
         <at 0 0 RCIM_Cubic>
         {arriveTangent=0.9268, interpMode=RCIM_Cubic, leaveTangent=0.9268, time=0.27, value=1}
         {arriveTangent=-1.365, interpMode=RCIM_Cubic, leaveTangent=-1.365, time=0.7207, value=0.668}
         <at 1.0011 0.002>
        ]}}})
    Scale RGB=&VectorFromCurve(VectorCurve=DataInterface() {propertyValues {
       XCurve {keys [<at 0 0>, <at 0.2 1.0067>, {arriveTangent=-1.2584, interpMode=RCIM_Cubic, leaveTangent=-1.2584, time=0.8, value=0.9431}, <at 1 0>]}
       YCurve {keys [<at 0 0>, <at 0.2 1.0067>, {arriveTangent=-1.2584, interpMode=RCIM_Cubic, leaveTangent=-1.2584, time=0.8, value=0.4309}, <at 1 0>]}
       ZCurve {keys [<at 0 0>, <at 0.2 1.0067>, {arriveTangent=-1.2584, interpMode=RCIM_Cubic, leaveTangent=-1.2584, time=0.8, value=0}, <at 1 0>]}
      }})
    ScaleA=true
    ScaleRGB=true
   ) {module=ScaleColor001}
   /SetVariables(Particles.Sprite2_Color:LinearColor=&MultiplyLinearColor(
     Linear Color=<rgba 3 0.6953 0.2344 0.3>
     Scale Factor=&MakeVector4(
      X=1
      Y=1
      Z=1
      W=&FloatFromCurve(FloatCurve=DataInterface() {propertyValues {Curve {keys [
           {arriveTangent=5.6712, interpMode=RCIM_Cubic, leaveTangent=5.6712, tangentMode=RCTM_User, time=0, value=0}
           {arriveTangent=2.0969, interpMode=RCIM_Cubic, leaveTangent=2.0969, tangentMode=RCTM_User, time=0.1089, value=0.7701}
           {arriveTangent=-1.6357, interpMode=RCIM_Cubic, leaveTangent=-1.6357, tangentMode=RCTM_User, time=1.0011, value=0.002}
          ]}}})
     )
    )) {isSetParameters=true}
  ]
  rendererClasses [&NiagaraSpriteRendererProperties]
  renderers [{rendererClass=&NiagaraSpriteRendererProperties}]
 }
 Fire_Hand {
  emitterUpdateScript [
   &EmitterState(Delay First Loop Only=true, Life Cycle Mode=::NewEnumerator1, Loop Delay=1, Loop Duration=1, UseLoopDelay=true)
   &SpawnBurst_Instantaneous(Spawn Count=8, Spawn Time=0.08, Use Spawn Probability=false) {enabled=false}
   &SpawnRate(SpawnRate=8, Use Spawn Probability=false)
  ]
  particleSpawnScript [
   &InitializeParticle(
    Color=<rgba 40 9.2708 3.125 1>
    Lifetime=&UniformRangedFloat(Minimum=0.43, Maximum=0.48)
    Module.Write Color=true
    Module.Write Lifetime=true
    Module.Write Mass=true
    Module.Write Material Random=true
    Module.Write Position=false
    Module.Write Scale=false
    Module.Write SpriteRotation=true
    Module.Write SpriteSize=true
    Position=&SimulationPosition()
    Sprite Rotation=&UniformRangedFloat(Minimum=0, Maximum=360)
    Sprite Size=&Vector2DFromFloat(Value=&UniformRangedFloat(Minimum=50, Maximum=70))
   )
   &DynamicMaterialParameters(
    Index 0 Param 2=&UniformRangedFloat(Minimum=0, Maximum=1)
    Index 0 Param 4=&UniformRangedFloat(Minimum=3, Maximum=5)
    Module.Param0WriteEnabled=false
    Module.Param1WriteEnabled=true
    Module.Param2WriteEnabled=false
    Module.Param3WriteEnabled=true
   )
   &AddVelocityFromPoint(
    Module.Use Falloff Distance=false
    Origin Offset=<0 0 -10>
    Velocity Origin:Vector3f=&SimulationPosition()
    Velocity Strength=80
   )
   &SkeletalMeshLocation(
    Bone Sampling Mode=::NewEnumerator4
    Inherit Velocity=false
    Interpolate Mesh Sampling=false
    MaskSpawnGroup=false
    Sampled Position Offset=&RandomRangeVector(Minimum=<-3 -3 -7>, Maximum=<3 3 -6>)
    Skeletal Mesh=$User.SkeletalMesh
    Use Sampled Mesh Attribute as Ribbon ID=false
    Write Sampled Mesh Attrributes to Particle Attributes=true
   )
  ]
  particleUpdateScript [
   &ParticleState()
   &ScaleSpriteSize(Scale Factor=&Vector2DFromFloat(Value=&FloatFromCurve(FloatCurve=DataInterface() {propertyValues {Curve {keys [
          {arriveTangent=1.3687, interpMode=RCIM_Cubic, leaveTangent=1.3687, tangentMode=RCTM_User, time=0.0002, value=0.15}
          {arriveTangent=0.3921, interpMode=RCIM_Cubic, leaveTangent=0.3921, tangentMode=RCTM_User, time=0.5914, value=0.5701}
          {arriveTangent=-0.1019, interpMode=RCIM_Cubic, leaveTangent=-0.1019, tangentMode=RCTM_User, time=0.9863, value=0.6136}
         ]}}})))
   /SetVariables(Particles.SubImageIndex=&FloatFromCurve(
     FloatCurve=DataInterface() {propertyValues {Curve {keys [
         {arriveTangent=2.1444, interpMode=RCIM_Cubic, leaveTangent=2.1444, tangentMode=RCTM_User, time=0, value=0}
         {arriveTangent=0.1938, interpMode=RCIM_Cubic, leaveTangent=0.1938, tangentMode=RCTM_User, time=1, value=1}
        ]}}}
     Scale Curve=40
    )) {isSetParameters=true, module=SetVariables_#2}
   &DynamicMaterialParameters(
    Index 0 Param 1=&FloatFromCurve(FloatCurve=DataInterface() {propertyValues {Curve {keys [<at 0 0>,<at 0.8 1>,<at 1.0011 1>]}}})
    Index 0 Param 2=2.5
    Index 0 Param 3=1
    Index 0 Param 4=3
    Module.Param0WriteEnabled=true
    Module.Param1WriteEnabled=false
    Module.Param2WriteEnabled=true
    Module.Param3WriteEnabled=false
   ) {module=DynamicMaterialParameters001}
   &GravityForce(Gravity=&MakeVector(
     X=0
     Y=0
     Z=&FloatFromCurve(FloatCurve=DataInterface() {propertyValues {Curve {keys [<at 0 0>,<at 0.6 0>,<at 0.7124 0.9779>,<at 1 1>]}}}, Scale Curve=250)
    ))
   &Drag(Drag=2.5, Use Linear Drag=true, Use Rotational Drag=false) {enabled=false}
   &SolveForcesAndVelocity(Clamp Velocity=false, Limit Acceleration=false)
   &ScaleColor(
    Scale RGB=&MakeVectorFromLinearColorRGB(Color=&ColorFromCurve(ColorCurve=DataInterface() {propertyValues {
        AlphaCurve {keys [<at 0 1 RCIM_Cubic>]}
        BlueCurve {keys [<at 0 0.8073 RCIM_Cubic>,<at 0.1964 0.4479>,<at 0.8699 0.0677 RCIM_Cubic>]}
        GreenCurve {keys [<at 0 2.1892 RCIM_Cubic>,<at 0.1964 1.1205>,<at 0.8699 0.1173 RCIM_Cubic>]}
        RedCurve {keys [<at 0 3 RCIM_Cubic>,<at 0.1964 2>,<at 0.8699 1 RCIM_Cubic>]}
       }}))
    ScaleA=true
    ScaleRGB=true
   ) {enabled=false}
   &SpriteRotationRate(Rotation Rate=150)
   &ScaleColor(
    Scale Alpha=&FloatFromCurve(FloatCurve=DataInterface() {propertyValues {Curve {keys [
         <at 0 0 RCIM_Cubic>
         {arriveTangent=0.9268, interpMode=RCIM_Cubic, leaveTangent=0.9268, time=0.27, value=1}
         {arriveTangent=-1.365, interpMode=RCIM_Cubic, leaveTangent=-1.365, time=0.7207, value=0.668}
         <at 1.0011 0.002>
        ]}}})
    Scale RGB=&VectorFromCurve(VectorCurve=DataInterface() {propertyValues {
       XCurve {keys [<at 0 0>, <at 0.2 1.0067>, {arriveTangent=-1.2584, interpMode=RCIM_Cubic, leaveTangent=-1.2584, time=0.8, value=0.9431}, <at 1 0>]}
       YCurve {keys [<at 0 0>, <at 0.2 1.0067>, {arriveTangent=-1.2584, interpMode=RCIM_Cubic, leaveTangent=-1.2584, time=0.8, value=0.4309}, <at 1 0>]}
       ZCurve {keys [<at 0 0>, <at 0.2 1.0067>, {arriveTangent=-1.2584, interpMode=RCIM_Cubic, leaveTangent=-1.2584, time=0.8, value=0}, <at 1 0>]}
      }})
    ScaleA=true
    ScaleRGB=true
   ) {module=ScaleColor001}
  ]
  rendererClasses [&NiagaraSpriteRendererProperties]
  renderers [{rendererClass=&NiagaraSpriteRendererProperties}]
 }
 Fire_LowerArm {
  emitterUpdateScript [
   &EmitterState()
   &SpawnRate(SpawnRate=5, Use Spawn Probability=false)
  ]
  particleSpawnScript [
   &InitializeParticle(
    Color=<rgba 40 9.2708 3.125 1>
    Lifetime=&UniformRangedFloat(Minimum=0.35, Maximum=0.5)
    Module.Write Color=true
    Module.Write Lifetime=true
    Module.Write Mass=true
    Module.Write Material Random=true
    Module.Write Position=false
    Module.Write Scale=false
    Module.Write SpriteRotation=true
    Module.Write SpriteSize=true
    Position=&SimulationPosition()
    Sprite Rotation=&UniformRangedFloat(Minimum=0, Maximum=360)
    Sprite Size=&Vector2DFromFloat(Value=&UniformRangedFloat(Minimum=33, Maximum=38))
   )
   &DynamicMaterialParameters(
    Index 0 Param 2=&UniformRangedFloat(Minimum=0, Maximum=1)
    Index 0 Param 4=&UniformRangedFloat(Minimum=3, Maximum=5)
    Module.Param0WriteEnabled=false
    Module.Param1WriteEnabled=true
    Module.Param2WriteEnabled=false
    Module.Param3WriteEnabled=true
   )
   &SkeletalMeshLocation(
    Bone / Socket Index=1
    Bone Sampling Mode=::NewEnumerator4
    Inherit Velocity=false
    Interpolate Mesh Sampling=false
    MaskSpawnGroup=false
    Sampled Position Offset=&RandomRangeVector(
     Minimum=&MakeVector(
      X=&FloatFromCurve(FloatCurve=DataInterface() {propertyValues {Curve {keys [<at 0 0>,<at 1 -18>]}}}, CurveIndex=$Emitter.Age)
      Y=0
      Z=0
     )
     Maximum=<5 0 0>
    )
    Sampled Position Offset Coordinate Space:ENiagaraMeshTransforms=::NewEnumerator3
    Skeletal Mesh=$User.SkeletalMesh
    Use Sampled Mesh Attribute as Ribbon ID=false
    Write Sampled Mesh Attrributes to Particle Attributes=true
   ) {enabled=false}
   &AddVelocity(Velocity=<0 0 110>) {enabled=false}
   /SetVariables(Particles.Sprite2_Size:Vector2f=<33 33>) {isSetParameters=true}
  ]
  particleUpdateScript [
   &ParticleState()
   &SkeletalMeshLocation(
    Bone / Socket Index=1
    Bone Sampling Mode=::NewEnumerator4
    Inherit Velocity=false
    MaskSpawnGroup=false
    Sampled Position Offset=&MakeVector(
     X=&FloatFromCurve(FloatCurve=DataInterface() {propertyValues {Curve {keys [<at 0 0>,<at 1 -25>]}}})
     Y=0
     Z=0
    )
    Sampled Position Offset Coordinate Space:ENiagaraMeshTransforms=::NewEnumerator3
    Skeletal Mesh=$User.SkeletalMesh
   ) {module=SkeletalMeshLocation001}
   &ScaleSpriteSize(Scale Factor=&Vector2DFromFloat(Value=&FloatFromCurve(FloatCurve=DataInterface() {propertyValues {Curve {keys [
          {arriveTangent=2.4033, interpMode=RCIM_Cubic, leaveTangent=2.4033, tangentMode=RCTM_User, time=0.0002, value=0.15}
          {arriveTangent=-0.0027, interpMode=RCIM_Cubic, leaveTangent=-0.0027, tangentMode=RCTM_User, time=0.6253, value=0.7425}
          {arriveTangent=-0.1019, interpMode=RCIM_Cubic, leaveTangent=-0.1019, tangentMode=RCTM_User, time=1.0041, value=0.7174}
         ]}}})))
   /SetVariables(Particles.SubImageIndex=&FloatFromCurve(
     FloatCurve=DataInterface() {propertyValues {Curve {keys [
         {arriveTangent=2.1444, interpMode=RCIM_Cubic, leaveTangent=2.1444, tangentMode=RCTM_User, time=0, value=0}
         {arriveTangent=0.1938, interpMode=RCIM_Cubic, leaveTangent=0.1938, tangentMode=RCTM_User, time=1, value=1}
        ]}}}
     Scale Curve=40
    )) {isSetParameters=true, module=SetVariables_#2}
   &DynamicMaterialParameters(
    Index 0 Param 1=&FloatFromCurve(FloatCurve=DataInterface() {propertyValues {Curve {keys [<at 0 0>,<at 0.8 1>,<at 1.0011 1>]}}})
    Index 0 Param 2=2.5
    Index 0 Param 3=1
    Index 0 Param 4=3
    Module.Param0WriteEnabled=true
    Module.Param1WriteEnabled=false
    Module.Param2WriteEnabled=true
    Module.Param3WriteEnabled=false
   ) {module=DynamicMaterialParameters001}
   &GravityForce(Gravity=&MakeVector(
     X=0
     Y=0
     Z=&FloatFromCurve(FloatCurve=DataInterface() {propertyValues {Curve {keys [<at 0 0>,<at 0.6 0>,<at 0.7124 0.9779>,<at 1 1>]}}}, Scale Curve=250)
    )) {enabled=false}
   &Drag(Drag=2.5, Use Linear Drag=true, Use Rotational Drag=false) {enabled=false}
   &SolveForcesAndVelocity(Clamp Velocity=false, Limit Acceleration=false)
   &ScaleColor(
    Scale RGB=&MakeVectorFromLinearColorRGB(Color=&ColorFromCurve(ColorCurve=DataInterface() {propertyValues {
        AlphaCurve {keys [<at 0 1 RCIM_Cubic>]}
        BlueCurve {keys [<at 0 0.8073 RCIM_Cubic>,<at 0.1964 0.4479>,<at 0.8699 0.0677 RCIM_Cubic>]}
        GreenCurve {keys [<at 0 2.1892 RCIM_Cubic>,<at 0.1964 1.1205>,<at 0.8699 0.1173 RCIM_Cubic>]}
        RedCurve {keys [<at 0 3 RCIM_Cubic>,<at 0.1964 2>,<at 0.8699 1 RCIM_Cubic>]}
       }}))
    ScaleA=true
    ScaleRGB=true
   ) {enabled=false}
   &SpriteRotationRate(Rotation Rate=150)
   &CameraOffset(Camera Offset Amount=10)
   &ScaleColor(
    Scale Alpha=&FloatFromCurve(FloatCurve=DataInterface() {propertyValues {Curve {keys [
         <at 0 0 RCIM_Cubic>
         {arriveTangent=0.9268, interpMode=RCIM_Cubic, leaveTangent=0.9268, time=0.27, value=1}
         {arriveTangent=-1.365, interpMode=RCIM_Cubic, leaveTangent=-1.365, time=0.7207, value=0.668}
         <at 1.0011 0.002>
        ]}}})
    Scale RGB=&VectorFromCurve(VectorCurve=DataInterface() {propertyValues {
       XCurve {keys [<at 0 0>, <at 0.2 1.0067>, {arriveTangent=-1.2584, interpMode=RCIM_Cubic, leaveTangent=-1.2584, time=0.8, value=0.9431}, <at 1 0>]}
       YCurve {keys [<at 0 0>, <at 0.2 1.0067>, {arriveTangent=-1.2584, interpMode=RCIM_Cubic, leaveTangent=-1.2584, time=0.8, value=0.4309}, <at 1 0>]}
       ZCurve {keys [<at 0 0>, <at 0.2 1.0067>, {arriveTangent=-1.2584, interpMode=RCIM_Cubic, leaveTangent=-1.2584, time=0.8, value=0}, <at 1 0>]}
      }})
    ScaleA=true
    ScaleRGB=true
   ) {module=ScaleColor001}
   /SetVariables(Particles.Sprite2_Color:LinearColor=&MultiplyLinearColor(
     Linear Color=<rgba 3 0.6953 0.2344 0.2>
     Scale Factor=&MakeVector4(
      X=1
      Y=1
      Z=1
      W=&FloatFromCurve(FloatCurve=DataInterface() {propertyValues {Curve {keys [
           {arriveTangent=5.6712, interpMode=RCIM_Cubic, leaveTangent=5.6712, tangentMode=RCTM_User, time=0, value=0}
           {arriveTangent=2.0969, interpMode=RCIM_Cubic, leaveTangent=2.0969, tangentMode=RCTM_User, time=0.1089, value=0.7701}
           {arriveTangent=-1.6357, interpMode=RCIM_Cubic, leaveTangent=-1.6357, tangentMode=RCTM_User, time=1.0011, value=0.002}
          ]}}})
     )
    )) {isSetParameters=true}
  ]
  rendererClasses [&NiagaraSpriteRendererProperties]
  renderers [
   {rendererClass=&NiagaraSpriteRendererProperties}
   {rendererClass=&NiagaraSpriteRendererProperties, rendererIndex=1}
  ]
 }
 Glow_Burst {
  emitterUpdateScript [
   &EmitterState(
    Life Cycle Mode=::NewEnumerator1
    Loop Behavior=::NewEnumerator1
    Loop Delay=1
    Loop Duration=1
    Reset Age On Awaken=true
    Scalability Mode=::NewEnumerator1
    UseLoopDelay=true
   )
   &SpawnBurst_Instantaneous(Spawn Count=1, Use Spawn Probability=false)
  ]
  particleSpawnScript [&InitializeParticle(
    Color=<rgba 1 0.2318 0.0781 0.35>
    Lifetime=0.35
    Module.Write Color=true
    Module.Write Lifetime=true
    Module.Write Mass=true
    Module.Write Material Random=true
    Module.Write Position=false
    Module.Write Scale=false
    Module.Write SpriteRotation=false
    Module.Write SpriteSize=true
    Position=&SimulationPosition()
    Sprite Size=&Vector2DFromFloat(Value=90)
   )]
  particleUpdateScript [
   &ParticleState(Kill Particles When Lifetime Has Elapsed=false, Let Infinitely Lived Particles Die When Emitter Deactivates=true)
   &SkeletalMeshLocation(
    Bone Sampling Mode=::NewEnumerator4
    Inherit Velocity=false
    Interpolate Mesh Sampling=false
    MaskSpawnGroup=false
    Sampled Position Offset=<0 0 10>
    Skeletal Mesh=$User.SkeletalMesh
    Use Sampled Mesh Attribute as Ribbon ID=false
    Write Sampled Mesh Attrributes to Particle Attributes=true
   )
   &SolveForcesAndVelocity(Clamp Velocity=false, Limit Acceleration=false)
   &ScaleSpriteSize(Scale Factor=&Vector2DFromFloat(Value=&FloatFromCurve(FloatCurve=DataInterface() {propertyValues {Curve {keys [
          {arriveTangent=2.7894, interpMode=RCIM_Cubic, leaveTangent=2.7894, tangentMode=RCTM_User, time=-0.0035, value=0.0142}
          {arriveTangent=0.9824, interpMode=RCIM_Cubic, leaveTangent=0.9824, time=0.3472, value=0.7522}
          <at 1 1>
         ]}}})))
   &CameraOffset(Camera Offset Amount=5) {enabled=false}
   &ScaleColor(
    Scale Alpha=&FloatFromCurve(
     FloatCurve=DataInterface() {propertyValues {Curve {keys [
         {arriveTangent=2.9001, interpMode=RCIM_Cubic, time=0, value=0}
         <at 0.1991 1.0186 RCIM_Cubic>
         {arriveTangent=-1.791, interpMode=RCIM_Cubic, leaveTangent=-5.8665, tangentMode=RCTM_User, time=1, value=0}
        ]}}}
     Scale Curve=0.5
    )
    ScaleA=true
    ScaleRGB=true
   ) {enabled=false}
   &ScaleColor(
    Scale Alpha=&FloatFromCurve(FloatCurve=DataInterface() {propertyValues {Curve {keys [
         {arriveTangent=5.4547, interpMode=RCIM_Cubic, leaveTangent=5.4547, tangentMode=RCTM_User, time=0, value=0}
         {arriveTangent=0.7125, interpMode=RCIM_Cubic, leaveTangent=0.7125, time=0.27, value=1}
         {arriveTangent=-1.6321, interpMode=RCIM_Cubic, leaveTangent=-1.6321, tangentMode=RCTM_User, time=0.6863, value=0.489}
         <at 1.0011 0.002>
        ]}}})
    Scale RGB=&VectorFromCurve(VectorCurve=DataInterface() {propertyValues {
       XCurve {keys [<at 0 0>, <at 0.2 1.0067>, {arriveTangent=-1.2584, interpMode=RCIM_Cubic, leaveTangent=-1.2584, time=0.8, value=0.9431}, <at 1 0>]}
       YCurve {keys [<at 0 0>, <at 0.2 1.0067>, {arriveTangent=-1.2584, interpMode=RCIM_Cubic, leaveTangent=-1.2584, time=0.8, value=0.4309}, <at 1 0>]}
       ZCurve {keys [<at 0 0>, <at 0.2 1.0067>, {arriveTangent=-1.2584, interpMode=RCIM_Cubic, leaveTangent=-1.2584, time=0.8, value=0}, <at 1 0>]}
      }})
    ScaleA=true
    ScaleRGB=true
   ) {module=ScaleColor001}
   &FN_Light_Attributes(
    Color=<rgba 1 0.3509 0.0917 1>
    Create Color Attribute=true
    Create Exponent Attribute=true
    Create Radius Attribute=true
    Exponent=8
    Radius=130
    Scale Color=&FloatFromCurve(
     FloatCurve=DataInterface() {propertyValues {Curve {keys [
         {arriveTangent=5.4547, interpMode=RCIM_Cubic, leaveTangent=5.4547, tangentMode=RCTM_User, time=0, value=0}
         {arriveTangent=0.7125, interpMode=RCIM_Cubic, leaveTangent=0.7125, time=0.27, value=1}
         {arriveTangent=-1.6321, interpMode=RCIM_Cubic, leaveTangent=-1.6321, tangentMode=RCTM_User, time=0.6863, value=0.489}
         <at 1.0011 0.002>
        ]}}}
     Scale Curve=17
    )
   )
  ]
  rendererClasses [&NiagaraSpriteRendererProperties, &NiagaraLightRendererProperties]
  renderers [
   {rendererClass=&NiagaraSpriteRendererProperties}
   {rendererClass=&NiagaraLightRendererProperties, rendererIndex=1}
  ]
 }
 Glow_Hand {
  emitterUpdateScript [
   &EmitterState(
    Life Cycle Mode=::NewEnumerator1
    Loop Behavior=::NewEnumerator1
    Loop Delay=1
    Loop Duration Mode=::NewEnumerator1
    Reset Age On Awaken=true
    Scalability Mode=::NewEnumerator1
    UseLoopDelay=true
   )
   &SpawnBurst_Instantaneous(Spawn Count=1, Use Spawn Probability=false)
  ]
  particleSpawnScript [&InitializeParticle(
    Color=<rgba 1 0.3605 0.0875 0.25>
    Module.Write Color=true
    Module.Write Lifetime=true
    Module.Write Mass=true
    Module.Write Material Random=true
    Module.Write Position=false
    Module.Write Scale=false
    Module.Write SpriteRotation=false
    Module.Write SpriteSize=true
    Position=&SimulationPosition()
    Sprite Size=&Vector2DFromFloat(Value=70)
   )]
  particleUpdateScript [
   &ParticleState(Kill Particles When Lifetime Has Elapsed=false, Let Infinitely Lived Particles Die When Emitter Deactivates=true)
   &SkeletalMeshLocation(
    Bone Sampling Mode=::NewEnumerator4
    Inherit Velocity=false
    Interpolate Mesh Sampling=false
    MaskSpawnGroup=false
    Sampled Position Offset=<0 0 10>
    Skeletal Mesh=$User.SkeletalMesh
    Use Sampled Mesh Attribute as Ribbon ID=false
    Write Sampled Mesh Attrributes to Particle Attributes=true
   )
   &SolveForcesAndVelocity(Clamp Velocity=false, Limit Acceleration=false)
   &ScaleSpriteSize(Scale Factor=&Vector2DFromFloat(Value=&FloatFromCurve(FloatCurve=DataInterface() {propertyValues {Curve {keys [<at -0.0035 0.0142 RCIM_Cubic>, {arriveTangent=0.9824, interpMode=RCIM_Cubic, leaveTangent=0.9824, time=0.3472, value=0.7522}, <at 1 1>]}}})))
   &CameraOffset(Camera Offset Amount=5) {enabled=false}
   &ScaleColor(
    Scale Alpha=&FloatFromCurve(
     FloatCurve=DataInterface() {propertyValues {Curve {keys [{arriveTangent=2.9001, interpMode=RCIM_Cubic, leaveTangent=2.9001, tangentMode=RCTM_User, time=0, value=0}, <at 1 1 RCIM_Cubic>]}}}
     Scale Curve=0.5
    )
    ScaleA=true
    ScaleRGB=true
   )
   &FN_Light_Attributes(
    Color=<rgba 1 0.3509 0.0917 1>
    Create Color Attribute=true
    Create Exponent Attribute=true
    Create Radius Attribute=true
    Exponent=8
    Radius=80
    Scale Color=&FloatFromCurve(FloatCurve=DataInterface() {propertyValues {Curve {keys [<at 0 0 RCIM_Cubic>,<at 0.6 1 RCIM_Cubic>]}}}, Scale Curve=7.5)
   )
  ]
  rendererClasses [&NiagaraSpriteRendererProperties, &NiagaraLightRendererProperties]
  renderers [
   {rendererClass=&NiagaraSpriteRendererProperties}
   {rendererClass=&NiagaraLightRendererProperties, rendererIndex=1}
  ]
 }
}
summary {
 systemName=NS_Thphoon_Emote
 emitters [
  {emitterName=Glow_Hand, rendererClasses [&NiagaraSpriteRendererProperties, &NiagaraLightRendererProperties]}
  {emitterName=Embers_Hand, rendererClasses [&NiagaraSpriteRendererProperties]}
  {emitterName=Fire_Hand, rendererClasses [&NiagaraSpriteRendererProperties]}
  {emitterName=Fire_Burst, rendererClasses [&NiagaraSpriteRendererProperties]}
  {emitterName=Glow_Burst, rendererClasses [&NiagaraSpriteRendererProperties, &NiagaraLightRendererProperties]}
  {emitterName=Fire_LowerArm, rendererClasses [&NiagaraSpriteRendererProperties]}
  {emitterName=Fire_Decal, rendererClasses [&NiagaraSpriteRendererProperties]}
 ]
 userVariables [User.SkeletalMesh {
   type=NiagaraDataInterfaceSkeletalMesh
   defaultValue {
    struct=NiagaraExt_VariableValue_DataInterface
    value {dataInterface=&self:NiagaraDataInterfaceSkeletalMesh_0}
   }
  }]
}
